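# Mono MetaSound source for the active loop sound of the Zero Point Dash crystal gadget.
# On play it triggers a single Wave Player on the Zero Point Crystal loop wave and sends that straight to the output.
# The source is flagged looping, though the player's Loop default holds what looks like a stray builder-result string.
# It uses crystal-specific attenuation, the gadgets and items sound class, and a concurrency override capping simultaneous instances.
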MetaSoundSource /WildEstate_ZPDashShared/Assets/Audio/MSS_WildEstate_ZeroPointDash_ActiveLoop
inputs [UE.Source.OnPlay {type=Trigger}]
interfaces [UE.OutputFormat.Mono:1.0, UE.Source:1.1]
outputs ["UE.OutputFormat.Mono.Audio:0" {type=Audio}]
pages [Default {
  input_defaults {UE.Source.OnPlay=null}
  nodes [
   InitVariable.Trigger()
   InitVariable.Trigger()
   InitVariable.Trigger()
   WavePlayer_0 = UE.Wave Player:Mono(Play=^UE.Source.OnPlay) {also {
     defaults {Loop="<MetaSoundBuilderResult.SUCCEEDED: 0>", Wave Asset=&/ZeroPointCrystal_Active_Loop_01}
     version [1,1]
    }}
  ]
  links [{dst="@UE.OutputFormat.Mono.Audio:0", src=WavePlayer_0.Out Mono}]
 }]
properties {
 assetImportData=&self:AssetImportData
 attenuationSettings=&/ATT_Fort_ZeroPointCrystal_ActiveLoop
 bLooping=true
 bOverrideConcurrency=true
 duration=10000
 maxDistance=6000
 sampleRate=48000
 soundClassObject=&Fort_Gadgets_Items
 volume=0.4
 concurrencyOverrides {maxCount=3}
}
