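# One-shot stereo MetaSound for the Zero Point Dash deactivate sound in WildEstate, classed as a gadget item sound.
# On play it layers a random, repeat-avoiding pick from three deactivate waves, pitched down randomly, with a fixed Zero Point Crystal power-down wave at its own random pitch.
# Both players feed a stereo mixer, and a trigger accumulator signals finished only after both end.
# Spatialization reuses the Zero Point Crystal active-loop attenuation asset.
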MetaSoundSource /WildEstate_ZPDashShared/Assets/Audio/MSS_WildEstate_ZeroPointDash_Deactivate
inputs [
 Deactivate Sounds {type=WaveAsset:Array}
 UE.Source.OnPlay {type=Trigger}
]
interfaces [UE.OutputFormat.Stereo:1.0, UE.Source.OneShot:1.0, UE.Source:1.1]
outputs [
 "UE.OutputFormat.Stereo.Audio:0" {type=Audio}
 "UE.OutputFormat.Stereo.Audio:1" {type=Audio}
 UE.Source.OneShot.OnFinished {type=Trigger}
]
pages [Default {
  input_defaults {
   UE.Source.OnPlay=null
   Deactivate Sounds [
    &/WildEstate_ZeroPointDash_Deactivate_01
    &/WildEstate_ZeroPointDash_Deactivate_02
    &/WildEstate_ZeroPointDash_Deactivate_03
   ]
  }
  nodes [
   ArrayRandomGet_0 = Array.Random Get:WaveAsset:Array(In Array=^Deactivate Sounds, Next=^UE.Source.OnPlay) {also {defaults {No Repeats=3}}}
   n1 = UE.RandomFloat(Next=@ArrayRandomGet_0.On Next) {also {defaults {Max=-3, Min=-1}}}
   WavePlayer_0 = UE.Wave Player:Stereo(Wave Asset=@ArrayRandomGet_0.Value, Play=@n1.On Next, Pitch Shift=@n1.Value) {also {version [1,1]}}
   n2 = UE.RandomFloat(Next=^UE.Source.OnPlay) {also {defaults {Max=2, Min=-1}}}
   WavePlayer_1 = UE.Wave Player:Stereo(Play=@n2.On Next, Pitch Shift=@n2.Value) {also {defaults {Wave Asset=&/ZeroPointCrystal_PowerDown_01}, version [1,1]}}
   AudioMixerAudioMixerStereo2_0 = "AudioMixer.Audio Mixer (Stereo, 2)"(In 0 L=@WavePlayer_0.Out Left, In 0 R=@WavePlayer_0.Out Right, In 1 L=@WavePlayer_1.Out Left, In 1 R=@WavePlayer_1.Out Right)
   TriggerAccumulatorTriggerAccumulate2_0 = "TriggerAccumulator.Trigger Accumulate (2)"(In 0=@WavePlayer_0.On Finished, In 1=@WavePlayer_1.On Finished)
  ]
  links [
   {dst="@UE.OutputFormat.Stereo.Audio:0", src=AudioMixerAudioMixerStereo2_0.Out L}
   {dst="@UE.OutputFormat.Stereo.Audio:1", src=AudioMixerAudioMixerStereo2_0.Out R}
   {dst="@UE.Source.OneShot.OnFinished", src=TriggerAccumulatorTriggerAccumulate2_0.Out}
  ]
 }]
properties {
 assetImportData=&self:AssetImportData
 attenuationSettings=&/ATT_Fort_ZeroPointCrystal_ActiveLoop
 maxDistance=6000
 numChannels=2
 outputFormat=Stereo
 sampleRate=48000
 soundClassObject=&Fort_Gadgets_Items
 volume=0.4
}
