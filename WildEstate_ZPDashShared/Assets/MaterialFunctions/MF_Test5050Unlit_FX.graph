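# Material function that fakes a partly unlit, flat-shaded look by splitting color between scaled BaseColor and emissive, using time-of-day tint parameters.
# Optional directional shading uses sky light direction and vertex normals to break up corners and suppress sun-facing base color, mask-blended with the flat version.
# Material proxies always get the flat version.
# The DONTTOUCHTHIS static switch gates directional shading and selects a second Enabled/Disabled attributes output.
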MaterialFunction /WildEstate_ZPDashShared/Assets/MaterialFunctions/MF_Test5050Unlit_FX
graph {nodes [
  n1 = GetMaterialAttributes(None=FunctionInput() {inputType=FunctionInput_MaterialAttributes}) {attributeGetTypes [#1, #2]}
  n2 = ComponentMask(None=CrossProduct(A=SkyAtmosphereLightDirection(), B=Constant3Vector() {constant=<rgba 0 0 1 0>})) {g=true, r=true}
  n3 = DotProduct(A=AppendVector(A=Normalize(VectorInput=@n2), B=Constant()), B=VertexNormalWS())
  n4 = ScalarParameter() {defaultValue=0.15, parameterName=Corner Break}
  n5 = Multiply(A=NamedRerouteUsage(), B=Add(A=Multiply(A=@n3, B=@n4) {constB=0.12}))
  n6 = ScalarParameter() {defaultValue=0.6, parameterName=Sun Supress}
  n7 = DotProduct(A=VertexNormalWS(), B=SkyAtmosphereLightDirection())
  n8 = Saturate(None=Multiply(A=Add(A=@n7)) {constB=0.5})
  n9 = Multiply(A=Multiply(A=@n5, B=LinearInterpolate(B=@n6, Alpha=@n8) {constA=1, constB=0})) {constB=0.75}
  n10 = Add(A=Multiply(A=@n5, B=NamedRerouteUsage()) {constB=0.2}, B=NamedRerouteUsage())
  n11 = NamedRerouteUsage()
  n12 = SetMaterialAttributes!(
   MaterialAttributes=@n1.MaterialAttributes
   BaseColor=Multiply(A=@n11) {constB=0.6}
   EmissiveColor=Multiply(A=@n11, B=NamedRerouteUsage()) {constB=0.2}
  ) {attributeSetTypes [#1, #2]}
  n13 = StaticSwitch(
   True=BlendMaterialAttributes(
    A=SetMaterialAttributes(MaterialAttributes=@n1.MaterialAttributes, BaseColor=@n9, EmissiveColor=@n10) {attributeSetTypes [#1, #2]}
    B=@n12
    Alpha=FunctionInput() {bUsePreviewValueAsDefault=true, inputName=DirectionalShadingMask, inputType=FunctionInput_Scalar, sortPriority=51}
   )
   False=@n12
   Value=FunctionInput(Preview=StaticBool()) {bUsePreviewValueAsDefault=true, inputName=useDirectionalShading, inputType=FunctionInput_StaticBool, sortPriority=50}
  )
  FunctionOutput(None=MaterialProxyReplace(Realtime=StaticSwitchParameter(True=@n13, False=@n12) {defaultValue=true, parameterName=DONTTOUCHTHIS}, MaterialProxy=@n12))
  FunctionOutput(None=StaticSwitchParameter(
    True=FunctionInput() {inputName=Enabled, inputType=FunctionInput_MaterialAttributes, sortPriority=10}
    False=FunctionInput() {inputName=Disabled, inputType=FunctionInput_MaterialAttributes, sortPriority=20}
   ) {defaultValue=true, parameterName=DONTTOUCHTHIS}) {outputName=EnabledDisabledOut, sortPriority=10}
  NamedRerouteDeclaration(None=@n1.EmissiveColor) {name=Emission, nodeColor=<rgba 1 0 0.3516 1>}
  NamedRerouteDeclaration(None=MaterialFunctionCall(
    Color_Morning=VectorParameter() {group=FlatShading, parameterName=Test_Morning, defaultValue=<rgba 0.2293 0.2167 0.2865 1>}.RGB
    Color_Day=VectorParameter() {group=FlatShading, parameterName=Test_Day, defaultValue=<rgba 0.276 0.276 0.276 1>}.RGB
    Color_Evening=VectorParameter() {group=FlatShading, parameterName=Test_Evening, defaultValue=<rgba 0.2969 0.2047 0.292 1>}.RGB
    Color_Night=VectorParameter() {group=FlatShading, parameterName=Test_Night, defaultValue=<rgba 0.2109 0.2253 0.25 1>}.RGB
   ) {materialFunction=&Color_Time_of_Day}.Result) {name=Unlit Emission, nodeColor=<rgba 1 0.1172 0 1>}
  NamedRerouteDeclaration(None=@n1.BaseColor) {name=Base Color, nodeColor=<rgba 0 0.2031 1 1>}
 ]}
