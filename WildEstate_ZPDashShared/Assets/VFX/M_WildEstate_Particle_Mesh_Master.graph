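# Masked, two-sided master material for WildEstate Niagara mesh, sprite and ribbon effects, with a stylized shading finish.
# Optional SubUV maps, color-chart tinting and hue shift build the surface, with dynamic parameters driving emissive, chart lookup, erosion and wobble.
# Advanced erosion blends in cracks and noisy, bounds-based dissolve with normal-push displacement.
# Optional interior glow and time-of-day emissive precede color quantize, flat and sun-direction shading; the outline branch is unconnected.
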Material /WildEstate_ZPDashShared/Assets/VFX/M_WildEstate_Particle_Mesh_Master
graph {
 properties {
  bUseMaterialAttributes=true
  bUsedWithNanite=true
  bUsedWithNiagaraMeshParticles=true
  bUsedWithNiagaraRibbons=true
  bUsedWithNiagaraSprites=true
  blendMode=BLEND_Masked
  opacityMaskClipValue=0.16
  pixelDepthOffsetMode=PDOM_Legacy
  translucencyLightingMode=TLM_VolumetricPerVertexDirectional
  twoSided=true
  usedShadingModels=Default Lit
 }
 nodes [
  n1 = ScalarParameter() {defaultValue=2, group=Emissive, parameterName=InteriorGlowContrast, sortPriority=52}
  n2 = VectorParameter() {group=Diffuse Color Picker, parameterName=World Space Color Picker Color Override, sortPriority=4, defaultValue=<rgba 1 1 1 1>}
  n3 = MaterialFunctionCall!(f2, f3=ObjectPositionWS(), f4) {materialFunction=/Game/Packages/Fortnite_UniqueMaterials_1/VectorComponentOps}
  n4 = ScalarParameter() {defaultValue=16.71, group=Diffuse Color Picker, parameterName="Color Change Frequency (WS)", sortPriority=7}
  n5 = DynamicParameter() {defaultValue=<rgba 1 1 0 1>, paramNames [Emissive Intensity, ColorChartUV, Erode, WPOScale]}
  n6 = Constant()
  n7 = TextureSampleParameter2D(UVs=LinearInterpolate(
    A=Divide(A=@n3.f3 Add, B=@n4)
    B=AppendVector(A=@n5.ColorChartUV, B=@n6)
    Alpha=ScalarParameter() {group=Diffuse Color Picker, parameterName=Use Dynamic Parameter for Color Chart, sortPriority=8}
   )) {group=Diffuse Color Picker, parameterName=Color Chart, sortPriority=5, texture=/Game/Global/Textures/Default/ColorSets/CarColors_Interiors}
  n8 = ScalarParameter() {defaultValue=1, group=Diffuse Color Picker, parameterName=Use Color Chart, sortPriority=3}
  n9 = TextureCoordinate()
  n10 = VectorParameter() {group=Diffuse, parameterName=Diffuse Tile, channelNames {a=SubUV Diff Y, b=SubUV Diff X, g=Diff Y, r=Diff X}, defaultValue=<rgba 1 1 1 1>}
  n11 = TextureSampleParameter2D(UVs=Multiply(A=@n9, B=AppendVector(A=@n10.Diff X, B=@n10.Diff X))) {
   group=Diffuse
   parameterName=Diffuse
   samplerSource=SSM_Wrap_WorldGroupSettings
   sortPriority=1
   texture=/Engine/EngineMaterials/DefaultDiffuse
  }
  n12 = TextureSampleParameterSubUV!(UVs=Multiply(A=@n9, B=AppendVector(A=@n10.SubUV Diff X, B=@n10.SubUV Diff X)), Apply View MipBias) {
   group=Diffuse
   parameterName=Diffuse_SubUV
   samplerSource=SSM_Wrap_WorldGroupSettings
   sortPriority=2
   texture=&DefaultDiffuse
  }
  n13 = ScalarParameter() {group=Diffuse, parameterName=Use SubUV Diffuse Texture}
  n14 = LinearInterpolate(A=@n11.RGB, B=@n12.RGB, Alpha=@n13)
  n15 = TextureSampleParameter2D() {
   group=Diffuse Color Picker
   parameterName=World Space Color Picker Mask
   samplerSource=SSM_Wrap_WorldGroupSettings
   sortPriority=1
   texture=&T_BlacksRGB
  }
  n16 = ChannelMaskParameter(None=@n15.RGBA) {
   group=Diffuse Color Picker
   maskChannel=Blue
   parameterName=World Space Color Picker Mask Channel
   sortPriority=2
   defaultValue=<rgba 0 0 1 0>
  }
  n17 = LinearInterpolate(
   A=Multiply(A=LinearInterpolate(A=@n2.RGB, B=@n7.RGB, Alpha=@n8), B=@n14)
   B=@n14
   Alpha=LinearInterpolate(
    A=@n16.R
    B=OneMinus(None=@n16.R)
    Alpha=ScalarParameter() {group=Diffuse Color Picker, parameterName=Invert Color Picker Mask, sortPriority=9}
   )
  )
  n18 = ScalarParameter() {group=Diffuse Color Picker, parameterName=Use Color Picker Coloring}
  n19 = MaterialFunctionCall(
   Hue Shift Percentage=ScalarParameter() {group=Diffuse, parameterName=Hue Shift Percentage, sortPriority=4}
   Texture=Multiply(
    A=LinearInterpolate(A=@n14, B=@n17, Alpha=@n18)
    B=ScalarParameter() {defaultValue=1, group=Diffuse, parameterName=Diff_Mult, sortPriority=3}
   )
  ) {materialFunction=&HueShift}
  n20 = ParticleColor()
  n21 = TextureCoordinate()
  n22 = VectorParameter() {
   group=Specular
   parameterName=Spec Masks Tile
   channelNames {a=SubUV Spec Y, b=SubUV Spec X, g=Spec Y, r=Spec X}
   defaultValue=<rgba 1 1 1 1>
  }
  n23 = TextureSampleParameter2D(UVs=Multiply(A=@n21, B=AppendVector(A=@n22.SubUV Spec X, B=@n22.SubUV Spec X))) {
   group=Specular
   parameterName=SpecMasks
   samplerSource=SSM_Wrap_WorldGroupSettings
   samplerType=SAMPLERTYPE_Masks
   texture=&T_Fortnite_Default_S
  }
  n24 = TextureSampleParameterSubUV(UVs=Multiply(A=@n21, B=AppendVector(A=@n22.Spec X, B=@n22.Spec X))) {
   group=Specular
   parameterName=SpecMasks_SubUV
   samplerSource=SSM_Wrap_WorldGroupSettings
   samplerType=SAMPLERTYPE_Masks
   sortPriority=2
   texture=&T_Fortnite_Default_S
  }
  n25 = MaterialFunctionCall(Color=LinearInterpolate(A=@n23.RGB, B=@n24.RGB, Alpha=ScalarParameter() {group=Specular, parameterName=Use SubUV SpecMask Texture}) {constAlpha=0}) {materialFunction=&SplitComponents}
  n26 = ScalarParameter() {group=Specular, parameterName=SpecRoughnessMin, sortPriority=1}
  n27 = ScalarParameter() {defaultValue=1, group=Specular, parameterName=SpecRoughnessMax, sortPriority=2}
  n28 = VectorParameter() {group=Emissive, parameterName="Emissive Color && Intensity", sortPriority=2}
  n29 = Multiply(A=@n19.Result, B=@n28.RGB)
  n30 = ScalarParameter() {group=Emissive, parameterName=UseDiffuseinEmissive, sortPriority=1}
  n31 = CameraVectorWS()
  n32 = AtmosphericLightVector()
  n33 = ScalarParameter() {group=Emissive, parameterName=Emissive Mask Fresnel Fake Slerp, sortPriority=5}
  n34 = LinearInterpolate(
   B=MaterialFunctionCall(
    Camera Vector=LinearInterpolate(A=@n31, B=@n32, Alpha=@n33)
    Power=ScalarParameter() {defaultValue=1, group=Emissive, parameterName=Emissive Mask Fresnel Exponent, sortPriority=5}
    Cheap contrast dark=ScalarParameter() {defaultValue=0.04, group=Emissive, parameterName=Emissive Mask Fresnel Min Value, sortPriority=5}
    Cheap contrast bright=ScalarParameter() {defaultValue=1, group=Emissive, parameterName=Emissive Mask Fresnel Max Value, sortPriority=5}
   ) {materialFunction=&Fresnel_Function}.Result
   Alpha=ScalarParameter() {group=Emissive, parameterName=Emissive Mask Fresnel, sortPriority=5}
  ) {constA=1}
  n35 = ScalarParameter() {
   desc=This option is there if you are using Emissive Mask Fresnel and you like to avoid using the constant value of 1 for lower end platform in replacement to the fresnel.
   group=Emissive
   parameterName=Disable Emissive Mask Fresnel Constant on Low End Platforms
  }
  n36 = Constant()
  n37 = If(A=@n35, "A > B"=@n36, "A == B"=@n36, "A < B"=Constant() {r=1}) {constB=0.99}
  n38 = TextureSampleParameter2D() {
   group=Emissive
   parameterName=Emissive Mask Texture
   samplerSource=SSM_Wrap_WorldGroupSettings
   sortPriority=3
   texture=&T_WhitesRGB
  }
  n39 = TextureSampleParameterSubUV() {
   group=Emissive
   parameterName=Emissive Mask SubUV
   samplerSource=SSM_Wrap_WorldGroupSettings
   sortPriority=2
   texture=&T_WhitesRGB
  }
  n40 = Multiply(
   A=ShadingPathSwitch(Default=QualitySwitch(Default=@n34, Low=@n37), Mobile=@n37) {desc=Migrated from FeatureLevelSwitch}
   B=LinearInterpolate(A=@n38.RGBA, B=@n39.RGBA, Alpha=ScalarParameter() {group=Emissive, parameterName=Use SubUV Emissive Mask Texture}) {constAlpha=0}
  ) {constA=1}
  n41 = LinearInterpolate(
   A=ChannelMaskParameter(None=@n40) {group=Emissive, parameterName=Emissive Mask Texture Channel, sortPriority=4}.R
   B=ComponentMask(None=@n40) {b=true, g=true, r=true}
   Alpha=ScalarParameter() {group=Emissive, parameterName=Emissive Mask Uses RGB, sortPriority=5}
  )
  n42 = Multiply(
   A=Multiply(
    A=LinearInterpolate(A=@n28.RGB, B=@n29, Alpha=@n30)
    B=DynamicParameter() {defaultValue=<rgba 1 1 0 1>, paramNames [Emissive Intensity, ColorChartUV, Erode, WPOScale]}.Emissive Intensity
   )
   B=LinearInterpolate(A=@n41, B=OneMinus(None=@n41), Alpha=ScalarParameter() {group=Emissive, parameterName=Invert Emissive Mask, sortPriority=6})
  )
  n43 = ParticleColor()
  n44 = Multiply(A=@n42, B=@n43.RGB)
  n45 = Add(A=@n42, B=@n43.RGB)
  n46 = ScalarParameter() {group=Emissive, parameterName=Particle Emissive Mult Or Add}
  n47 = ScalarParameter() {defaultValue=1, group=Diffuse, parameterName=Diffuse Alpha Blend, sortPriority=3}
  n48 = LinearInterpolate(A=@n11.A, B=@n12.A, Alpha=@n13)
  n49 = Multiply(
   A=TextureCoordinate()
   B=ComponentMask(None=VectorParameter() {group=Opacity, parameterName=Opacity Mask UV Scale, sortPriority=3, defaultValue=<rgba 1 1 0 0>}.RGB) {g=true, r=true}
  )
  n50 = TextureSampleParameter2D(UVs=LinearInterpolate(
    A=@n49
    B=MaterialFunctionCall(XY=@n49) {materialFunction=&Swizzle}.YX
    Alpha=ScalarParameter() {group=Opacity, parameterName=Swizzle Opacity Mask UVs, sortPriority=4}
   )) {
   group=Opacity
   parameterName=Opacity_Mask_Texture
   samplerSource=SSM_Wrap_WorldGroupSettings
   sortPriority=1
   texture=&T_WhitesRGB
  }
  n51 = Multiply(
   A=Multiply(A=ParticleColor().A, B=LinearInterpolate(B=@n48, Alpha=@n47) {constA=1})
   B=ChannelMaskParameter(None=@n50.RGBA) {group=Opacity, parameterName=Opacity_Mask Channel Mask, sortPriority=2}.R
  )
  n52 = Multiply(
   A=MaterialFunctionCall(Distance=ScalarParameter() {group=Opacity, parameterName=Near Camera Fade Distance}) {materialFunction=&Near_Camera_Fade}.Result
   B=@n51
  )
  n53 = DepthFade(Opacity=@n52, FadeDistance=ScalarParameter() {defaultValue=25, group=Opacity, parameterName=DepthFadeDistance})
  n54 = TextureCoordinate()
  n55 = VectorParameter() {
   group=Normal
   parameterName=Normal Tile
   channelNames {a=SubUV normal Y, b=SubUV normal X, g=Normal Y, r=Normal X}
   defaultValue=<rgba 1 1 1 1>
  }
  n56 = TextureSampleParameter2D(UVs=Multiply(A=@n54, B=AppendVector(A=@n55.Normal X, B=@n55.Normal X))) {
   group=Normal
   parameterName=Normals
   samplerSource=SSM_Wrap_WorldGroupSettings
   samplerType=SAMPLERTYPE_Normal
   texture=/Engine/EngineMaterials/FlatNormal
  }
  n57 = AppendVector(A=@n55.SubUV normal X, B=@n55.SubUV normal X)
  n58 = LinearInterpolate(
   A=@n56.RGB
   B=TextureSampleParameterSubUV(UVs=Multiply(A=@n54, B=@n57)) {group=Normal, parameterName=Normal_SubUV, samplerType=SAMPLERTYPE_Normal, texture=&FlatNormal}.RGB
   Alpha=ScalarParameter() {group=Normal, parameterName=Use SubUV Normal Texture}
  ) {constAlpha=0}
  n59 = ScalarParameter() {group=Normal, parameterName=Normal Flatness}
  n60 = Multiply(
   A=MaterialFunctionCall!(
    Speed=ScalarParameter() {defaultValue=0.5, group=WorldPositionOffset, parameterName=WPO_Speed, sortPriority=1}
    LargeWaveScle=ScalarParameter() {defaultValue=1, group=WorldPositionOffset, parameterName=WPO_LargeWaveScale, sortPriority=2}
    WaveScale=ScalarParameter() {defaultValue=0.5, group=WorldPositionOffset, parameterName=WPO_WaveScale, sortPriority=3}
   ) {materialFunction=/Game/Effects/Fort_Effects/Materials/_Functions/Jiggle_WPO_VertColor_Red_3Vec}.Result
   B=DynamicParameter() {defaultValue=<rgba 1 1 0 1>, paramNames [Emissive Intensity, ColorChartUV, Erode, WPOScale]}.WPOScale
  )
  n61 = ScalarParameter() {group=WorldPositionOffset, parameterName=Use WPO}
  n62 = MaterialFunctionCall(MA In=MakeMaterialAttributes!(
    BaseColor=Multiply(A=@n19.Result, B=@n20.RGB)
    Metallic=@n25.G
    Specular=@n25.G
    Roughness=LinearInterpolate(A=@n26, B=@n27, Alpha=@n25.B)
    Anisotropy
    EmissiveColor=LinearInterpolate(A=@n44, B=@n45, Alpha=@n46)
    Opacity=StaticSwitchParameter(True=@n53, False=@n52) {group=Opacity, parameterName=UseDepthFade}
    OpacityMask=@n51
    Normal=MaterialFunctionCall(Normal=@n58, Flatness=@n59) {materialFunction=&FlattenNormal}.Result
    Tangent
    WorldPositionOffset=LinearInterpolate(B=@n60, Alpha=@n61)
    SubsurfaceColor
    ClearCoat
    ClearCoatRoughness
    AmbientOcclusion
    Refraction
    CustomizedUVs_0
    CustomizedUVs_1
    CustomizedUVs_2
    CustomizedUVs_3
    CustomizedUVs_4
    CustomizedUVs_5
    CustomizedUVs_6
    CustomizedUVs_7
    PixelDepthOffset
    ShadingModel
    Displacement
   )) {materialFunction=&MF_CreativeMoveTool}
  n63 = GetMaterialAttributes(None=@n62.MA Out) {attributeGetTypes [
    #1
    #2
    #3
    #4
    #5
   ]}
  n64 = VectorParameter() {group=Erosion, parameterName=Crack Color, sortPriority=11, defaultValue=<rgba 0.15 0.15 0.15 0>}
  n65 = LinearInterpolate(
   A=TextureCoordinate()
   B=TextureCoordinate() {coordinateIndex=1}
   Alpha=ScalarParameter() {defaultValue=1, group=Erosion, parameterName=Use UV1 Crack Texture, sortPriority=7}
  )
  n66 = TextureObjectParameter() {group=Erosion, parameterName=Crack, sortPriority=6, texture=/Game/Global/Textures/Default/Masks_and_Gradients/Cracks_02}
  n67 = ScalarParameter() {defaultValue=350, group=Erosion, parameterName=Crack World Aligned Texture Size, sortPriority=9}
  n68 = WorldPosition() {worldPositionShaderOffset=WPT_ExcludeAllShaderOffsets}
  n69 = Multiply(
   A=LinearInterpolate(
    A=TextureSample(UVs=@n65, Tex=@n66) {texture=&Cracks_02}.RGB
    B=MaterialFunctionCall(TextureObject=@n66, TextureSize=@n67, WorldPosition=@n68.XYZ) {materialFunction=&WorldAlignedTexture}.XYZ Texture
    Alpha=ScalarParameter() {defaultValue=1, group=Erosion, parameterName=Use WorldAligned Crack Texture, sortPriority=8}
   )
   B=ScalarParameter() {defaultValue=1, group=Erosion, parameterName=CrackSize, sortPriority=10}
  )
  n70 = ComponentMask(None=@n69) {r=true}
  n71 = Add(A=ObjectRadius()) {constB=0.001}
  n72 = Add(
   A=Subtract(A=WorldPosition() {worldPositionShaderOffset=WPT_ExcludeAllShaderOffsets}.XYZ, B=ObjectPositionWS()) {desc=Doesn't break with different locations}
   B=@n71
  )
  n73 = Multiply(A=@n71) {constB=2}
  n74 = Saturate(None=ComponentMask(None=Divide(A=@n72, B=@n73)) {b=true})
  n75 = LinearInterpolate(
   A=Constant() {r=0.5}
   B=LinearInterpolate(A=OneMinus(None=@n74), B=@n74, Alpha=ScalarParameter() {defaultValue=1, group=Erosion, parameterName=BoundRampDirection, sortPriority=5})
   Alpha=ScalarParameter() {defaultValue=1, group=Erosion, parameterName=Use Bounds in Erosion, sortPriority=14}
  )
  n76 = ComponentMask(None=@n69) {b=true}
  n77 = Multiply(A=@n76, B=ScalarParameter() {defaultValue=0.5, group=Erosion, parameterName=CrackAffectMask, sortPriority=5}) {constB=0.25}
  n78 = Multiply(
   A=TextureCoordinate() {coordinateIndex=1}
   B=ComponentMask(None=VectorParameter() {group=Erosion, parameterName=Noise_UV_Scale, sortPriority=4, defaultValue=<rgba 2 2 0 0>}.RGB) {g=true, r=true}
  )
  n79 = TextureSampleParameter2D(UVs=LinearInterpolate(
    A=@n78
    B=MaterialFunctionCall(XY=@n78) {materialFunction=&Swizzle}.YX
    Alpha=ScalarParameter() {group=Erosion, parameterName=Swizzle Erosion Noise, sortPriority=5}
   )) {
   group=Erosion
   parameterName=Erosion Noise
   samplerType=SAMPLERTYPE_LinearColor
   sortPriority=3
   texture=/Game/Characters/Player/Male/Medium/Bodies/M_MED_SRDrift/Textures/T_ElectricNoise
  }
  n80 = ScalarParameter() {defaultValue=1, group=Erosion, parameterName=NoiseAffectMask, sortPriority=2}
  n81 = Subtract(
   A=Add(A=Subtract(A=@n75, B=@n77), B=Multiply(A=@n79.R, B=@n80))
   B=DynamicParameter() {defaultValue=<rgba 1 1 0 1>, paramNames [Emissive Intensity, ColorChartUV, Erode, WPOScale]}.Erode
  )
  n82 = Multiply(A=@n81) {constB=-1}
  n83 = Saturate(None=Add(A=@n82) {constB=0.5})
  n84 = ScalarParameter() {defaultValue=0.1, group=Erosion, parameterName=Crack Specular, sortPriority=12}
  n85 = ScalarParameter() {defaultValue=1, group=Erosion, parameterName=Crack Roughness, sortPriority=13}
  n86 = Multiply(
   A=Multiply(A=VertexNormalWS(), B=Saturate(None=@n82))
   B=ScalarParameter() {defaultValue=50, group=Erosion, parameterName=Erosion_WPO_Strength, sortPriority=1}
  ) {constB=100}
  n87 = Constant()
  n88 = SetMaterialAttributes!(
   MaterialAttributes=@n63.MaterialAttributes
   Base Color=LinearInterpolate(A=LinearInterpolate(A=@n63.Base Color, B=@n64.RGB, Alpha=@n70), Alpha=@n83) {constB=0.1}
   Specular=LinearInterpolate(A=LinearInterpolate(A=@n63.Specular, B=@n84, Alpha=@n76), Alpha=@n83) {constB=0.15}
   Roughness=LinearInterpolate(A=LinearInterpolate(A=@n63.Roughness, B=@n85, Alpha=@n76), Alpha=@n83)
   Opacity Mask=Multiply(A=@n63.Opacity Mask, B=@n81)
   World Position Offset=Add(A=@n63.World Position Offset, B=ShadingPathSwitch(Default=@n86, Mobile=@n87))
  ) {attributeSetTypes [
    #1
    #2
    #3
    #4
    #5
   ]}
  n89 = ScalarParameter() {group=Erosion, parameterName=Use Advanced Erosion}
  n90 = GetMaterialAttributes(None=MaterialFunctionCall!(In=BlendMaterialAttributes(A=@n62.MA Out, B=@n88, Alpha=@n89)) {materialFunction=/Game/Effects/Niagara/MasterMaterials/Functions/MF_FX_MAEmissiveAdjustments}.Result) {attributeGetTypes [#6, #1, #7]}
  n91 = Transform(None=@n90.Normal)
  n92 = ScalarParameter() {defaultValue=2, group=Emissive, parameterName=InteriorGlowAmount, sortPriority=52}
  n93 = LinearInterpolate(
   A=@n92
   B=Multiply(A=@n92) {constB=1.25}
   Alpha=MaterialFunctionCall(Speed=ScalarParameter() {defaultValue=0.5, group=Emissive, parameterName=InteriorGlowSpeed, sortPriority=52}) {materialFunction=&DebugTimeSine}.Result
  )
  n94 = Multiply(A=Power(Base=Fresnel(ExponentIn=@n1, Normal=@n91), Exp=@n1), B=@n93)
  n95 = VertexNormalWS()
  n96 = CameraVectorWS()
  n97 = Multiply(A=Power(Base=DotProduct(A=@n95, B=@n96), Exp=@n1), B=@n93)
  n98 = Saturate(None=LinearInterpolate(A=@n94, B=Add(A=@n97, B=@n94), Alpha=ScalarParameter() {group=Emissive, parameterName=InteriorGlowInvert, sliderMax=1, sortPriority=52}))
  n99 = Constant() {r=3}
  Divide(A=Floor(None=Multiply(A=@n98, B=@n99) {constB=5}), B=Subtract(A=@n99))
  n100 = CameraVectorWS()
  n101 = PixelNormalWS()
  n102 = VectorParameter() {group=Emissive, parameterName=InteriorGlowColor, sortPriority=51, defaultValue=<rgba 0.491 0 0.7011 1>}
  n103 = StaticSwitchParameter(True=LinearInterpolate(A=@n90.Emissive Color, B=@n102.RGB, Alpha=@n98), False=@n90.Emissive Color) {group=Emissive, parameterName=UseInteriorGlow, sortPriority=50}
  n104 = VectorParameter() {group=Emissive, parameterName=TOD_Scalar, channelNames {a=Night, b=Evening, g=Day, r=Morning}, defaultValue=<rgba 1 1 1 1>}
  n105 = MaterialFunctionCall(Value_Morning=@n104.Morning, Value_Day=@n104.Morning, Value_Evening=@n104.Morning, Value_Night=@n104.Morning) {materialFunction=&Scalar_Time_of_Day}
  n106 = SetMaterialAttributes!(
   MaterialAttributes=@n90.MaterialAttributes
   Normal=@n91
   Emissive Color=EyeAdaptationInverse(LightValueInput=Multiply(A=@n103, B=@n105.Result)).EyeAdaptationInverse
  ) {attributeSetTypes [#6, #7]}
  n107 = GetMaterialAttributes(None=@n106) {attributeGetTypes [
    #2
    #3
    #1
    #7
    #6
   ]}
  n108 = VectorParameter() {group=Flat Lighting, parameterName=Outline Color, defaultValue=<rgba 0.01 0.01 0.01 1>}
  If(
   A=Saturate(None=DotProduct(A=@n100, B=@n101))
   B=ScalarParameter() {defaultValue=0.15, group=Flat Lighting, parameterName=Outline Thickness}
   "A > B"=@n107.Base Color
   "A < B"=@n108.RGB
  )
  n109 = ScalarParameter() {group=Diffuse, parameterName=HueShiftAmount, sortPriority=5}
  MaterialFunctionCall(Hue Shift Percentage=@n109) {materialFunction=&HueShift}
  MaterialFunctionCall(Hue Shift Percentage=@n109) {materialFunction=&HueShift}
  n110 = NamedRerouteUsage()
  n111 = Round(None=Fresnel(ExponentIn=ScalarParameter() {defaultValue=2, group=Flat Lighting, parameterName=Outline Fresnel}, Normal=VertexNormalWS()) {baseReflectFraction=0, exponent=4})
  SetMaterialAttributes!(
   MaterialAttributes=@n107.MaterialAttributes
   Specular=LinearInterpolate(A=@n107.Specular, B=Constant(), Alpha=@n110)
   Roughness=LinearInterpolate(A=@n107.Roughness, B=Constant() {r=1}, Alpha=@n110)
   Emissive Color=LinearInterpolate(A=@n107.Emissive Color, Alpha=@n110) {constB=0}
   Base Color=LinearInterpolate(A=@n107.Base Color, B=@n108.RGB, Alpha=NamedRerouteDeclaration(None=@n111) {name=Outline, nodeColor=<rgba 0 1 0.9063 1>})
  ) {attributeSetTypes [
    #2
    #3
    #7
    #1
   ]}
  n112 = MaterialFunctionCall!(
   Material In=MaterialFunctionCall!(In=@n106) {materialFunction=/WildEstate_ZPDashShared/Assets/MaterialFunctions/M_WildEstate_QantizeColor_FX}.Result
   UseFlatShading=StaticBoolParameter() {group=FlatShading, parameterName=UseFlatShading_All, sortPriority=1}
   IsWorldSpace=StaticBool() {value=true}
   OverrideUnlitShadinBaseColorDarkening
   OverrideUnlitShadinBaseColorEmissive
  ) {materialFunction=/WildEstate_ZPDashShared/Assets/MaterialFunctions/MF_WildEstate_FlatShading_FX}
  n113 = GetMaterialAttributes(None=MaterialFunctionCall!(
    In=MaterialFunctionCall!(In=@n112.Result) {materialFunction=/WildEstate_ZPDashShared/Assets/MaterialFunctions/M_WildEstate_SunDirectionShading_FX}.Result
    Enabled
    Disabled
    useDirectionalShading=StaticBool() {value=true}
    DirectionalShadingMask
   ) {materialFunction=/WildEstate_ZPDashShared/Assets/MaterialFunctions/MF_Test5050Unlit_FX}.Result) {attributeGetTypes [#6]}
  SetMaterialAttributes!(
   MaterialAttributes=@n113.MaterialAttributes
   Normal=MaterialProxyReplace(
    Realtime=@n113.Normal
    MaterialProxy=Transform(None=@n113.Normal) {transformSourceType=TRANSFORMSOURCE_World, transformType=TRANSFORM_Tangent}
   )
  ) {attributeSetTypes [#6]}
  StaticBool() {value=true}
 ]
}
